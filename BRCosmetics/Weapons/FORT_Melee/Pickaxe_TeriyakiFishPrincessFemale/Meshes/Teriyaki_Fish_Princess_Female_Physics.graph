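# Physics asset for the Teriyaki Fish Princess Female two-handed pickaxe cosmetic, defining bodies and joints for its handle, fish, petals and wings.
# Left and right wings, each with upper and lower segments, branch from a central wing body locked to the handle and swing mainly on one axis.
# The fish tail forms a two-link chain and six petals fan off one hub, all tightly limited.
# Neither fish nor petal hub has a constraint to the handle.
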PhysicsAsset /BRCosmetics/Weapons/FORT_Melee/Pickaxe_TeriyakiFishPrincessFemale/Meshes/Teriyaki_Fish_Princess_Female_Physics
bodies {
 Melee_TwoHanded_Handle [{shapeType=Capsule, length=95.0313, radius=1.6635, center=<0.2676 -1.0021 27.7574>, rotation=<rot -0.0012 -90 0.1658>}]
 dyn_fish [{shapeType=Capsule, length=19.7687, radius=3.3856, center=<-0.7031 0.2122 2.5028>, rotation=<rot -1.6963 -89.9643 -0.2814>}]
 dyn_fish_tail_01 [{shapeType=Capsule, length=4.3019, radius=2.8097, center=<-0.5261 -0.2046 -3.153>, rotation=<rot 0 179.1074 -0.2141>}]
 dyn_fish_tail_02 [{shapeType=Capsule, length=6.7008, radius=3.1197, center=<-0.5199 0.2195 -4.9411>, rotation=<rot 0 -179.8485 -87.9412>}]
 dyn_petal [{shapeType=Box, extentX=3.5, extentY=20, extentZ=16, center=<-5.318 0 0>}]
 dyn_petal_01 [{shapeType=Capsule, length=2, radius=1.8676, center=<-1.3827 0.101 1.7583>, rotation=<rot 0 180 1.7274>}]
 dyn_petal_02 [{shapeType=Capsule, length=2.1824, radius=1.8676, center=<-1.3828 0.106 1.2045>, rotation=<rot 0 179.9999 -178.4448>}]
 dyn_petal_03 [{shapeType=Capsule, length=1.6824, radius=1.8676, center=<-1.3827 -0.1198 1.4615>, rotation=<rot 0 180 -1.883>}]
 dyn_petal_04 [{shapeType=Capsule, length=1.6948, radius=1.8414, center=<-1.4567 0.1765 1.4484>, rotation=<rot 0 0.0536 -178.9169>}]
 dyn_petal_05 [{shapeType=Capsule, length=1.9779, radius=1.7114, center=<-1.1332 0.4278 1.3667>, rotation=<rot 2.0146 79.1814 178.3554>}]
 dyn_petal_06 [{shapeType=Capsule, length=2.1334, radius=1.8216, center=<-1.4295 -0.0381 1.3711>, rotation=<rot -8.0545 174.4205 6.0148>}]
 dyn_wing [{shapeType=Capsule, length=19.0364, radius=0.5, center=<-0.5622 0.4459 -3.5113>, rotation=<rot 0.291 -90.4173 3.3304>}]
 dyn_wing_l [{shapeType=Box, extentX=2, extentY=13.5, extentZ=25, center=<-0.42 -4.0519 -1.4324>}]
 dyn_wing_l_dn [{shapeType=Box, extentX=1.5, extentY=23.5, extentZ=15.5, center=<0.5302 -3.2671 0.0309>, rotation=<rot 0 0 4.5498>}]
 dyn_wing_l_up [{shapeType=Box, extentX=1.5, extentY=23, extentZ=22, center=<0.4107 -3.6684 -1.8696>, rotation=<rot 0 0 -5.7358>}]
 dyn_wing_r [{shapeType=Box, extentX=2, extentY=13, extentZ=25, center=<-0.3124 -4.3366 1.2948>}]
 dyn_wing_r_dn [{shapeType=Box, extentX=1.5, extentY=23.5, extentZ=15.5, center=<0.9605 -3.7379 -1.0933>, rotation=<rot 0 0 -12.1588>}]
 dyn_wing_r_up [{shapeType=Box, extentX=1.5, extentY=23.5, extentZ=21, center=<0.0978 -4.8509 1.5225>, rotation=<rot 0 0 1.1347>}]
}
constraints [
 {bone1Name=dyn_wing_l, bone2Name=dyn_wing, swing1LimitDegrees=25, swing2LimitDegrees=1, twistLimitDegrees=1}
 {bone1Name=dyn_wing_r, bone2Name=dyn_wing, swing1LimitDegrees=25, swing2LimitDegrees=1, twistLimitDegrees=1}
 {bone1Name=dyn_fish_tail_01, bone2Name=dyn_fish, swing1LimitDegrees=5, swing2LimitDegrees=20, twistLimitDegrees=5}
 {bone1Name=dyn_fish_tail_02, bone2Name=dyn_fish_tail_01, swing1LimitDegrees=5, swing2LimitDegrees=20, twistLimitDegrees=5}
 {bone1Name=dyn_petal_01, bone2Name=dyn_petal, swing1LimitDegrees=5, swing2LimitDegrees=20, twistLimitDegrees=5}
 {bone1Name=dyn_petal_02, bone2Name=dyn_petal, swing1LimitDegrees=5, swing2LimitDegrees=20, twistLimitDegrees=5}
 {bone1Name=dyn_petal_03, bone2Name=dyn_petal, swing1LimitDegrees=5, swing2LimitDegrees=20, twistLimitDegrees=5}
 {bone1Name=dyn_petal_04, bone2Name=dyn_petal, swing1LimitDegrees=5, swing2LimitDegrees=20, twistLimitDegrees=5}
 {bone1Name=dyn_petal_05, bone2Name=dyn_petal, swing1LimitDegrees=5, swing2LimitDegrees=20, twistLimitDegrees=5}
 {bone1Name=dyn_petal_06, bone2Name=dyn_petal, swing1LimitDegrees=5, swing2LimitDegrees=20, twistLimitDegrees=5}
 {bone1Name=dyn_wing_l_up, bone2Name=dyn_wing_l, swing1LimitDegrees=20, swing2LimitDegrees=1, twistLimitDegrees=1}
 {bone1Name=dyn_wing_l_dn, bone2Name=dyn_wing_l, swing1LimitDegrees=20, swing2LimitDegrees=1, twistLimitDegrees=1}
 {bone1Name=dyn_wing_r_up, bone2Name=dyn_wing_r, swing1LimitDegrees=20, swing2LimitDegrees=1, twistLimitDegrees=1}
 {bone1Name=dyn_wing_r_dn, bone2Name=dyn_wing_r, swing1LimitDegrees=20, swing2LimitDegrees=1, twistLimitDegrees=1}
 {
  bone1Name=dyn_wing
  bone2Name=Melee_TwoHanded_Handle
  swing1LimitDegrees=45
  swing1Motion=Locked
  swing2LimitDegrees=45
  swing2Motion=Locked
  twistLimitDegrees=45
  twistMotion=Locked
 }
]
